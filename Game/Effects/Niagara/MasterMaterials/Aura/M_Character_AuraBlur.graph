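# Translucent unlit Niagara sprite and mesh-particle material that draws an animated, noise-distorted character aura.
# Two panning noise layers, warped by secondary noise and randomized per particle, vertically displace two blended half-sphere mask samples.
# Opacity combines that mask with particle alpha, V-edge, view-angle and depth fades.
# Emissive mixes tint, noise and oversaturated scrolling color, multiplied by particle color, with a scalable clock driving all panning.
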Material /Game/Effects/Niagara/MasterMaterials/Aura/M_Character_AuraBlur
graph {
 properties {
  bUsedWithNiagaraMeshParticles=true
  bUsedWithNiagaraSprites=true
  blendMode=BLEND_Translucent
  pixelDepthOffsetMode=PDOM_Legacy
  shadingModel=MSM_Unlit
  usedShadingModels=Unlit
 }
 nodes [
  n1 = ParticleColor()
  n2 = TextureCoordinate()
  n3 = TextureCoordinate()
  n4 = TextureSampleParameter2D(UVs=Panner(Coordinate=TextureCoordinate(), Time=NamedRerouteUsage()) {speedY=0.1}) {
   parameterName=SecondNoiseTexture
   samplerType=SAMPLERTYPE_Masks
   texture=/Game/Characters/Player/Male/Medium/Bodies/M_MED_Storm_Soldier/Textures/ElectricityNoise_M
  }
  n5 = Multiply(A=ComponentMask(None=@n4.RGB) {g=true, r=true}, B=ScalarParameter() {parameterName=SecondNoise_Intensity}) {constB=0}
  n6 = Add(A=TextureCoordinate() {vTiling=0.3}, B=@n5)
  n7 = ParticleRandom()
  n8 = Multiply(
   A=AppendVector(A=ScalarParameter() {parameterName=NoiseTexture_SpeedU}, B=ScalarParameter() {defaultValue=0.75, parameterName=NoiseTexture_SpeedV})
   B=ScalarParameter() {defaultValue=1, parameterName=NoiseTexture_AllSpeed}
  )
  n9 = TextureSampleParameter2D(UVs=Panner(Coordinate=Add(A=@n6, B=@n7), Time=NamedRerouteUsage(), Speed=@n8) {speedY=1}) {parameterName=NoiseTexture01, samplerType=SAMPLERTYPE_LinearColor, texture=&T_NoisePack_LinearColor}
  n10 = LinearInterpolate(Alpha=@n9.R) {constA=-1, constB=0.5}
  n11 = ScalarParameter() {defaultValue=0.2, parameterName=Noise_Intensity}
  n12 = AppendVector(A=Add(A=ComponentMask(None=@n2) {r=true}) {constB=0}, B=Add(A=ComponentMask(None=@n3) {g=true}, B=Multiply(A=@n10, B=@n11) {constB=0.2}))
  n13 = Constant2Vector() {g=0.1}
  n14 = TextureCoordinate()
  n15 = Add(A=@n5, B=TextureCoordinate() {vTiling=-0.3})
  n16 = TextureSampleParameter2D(UVs=Panner(Coordinate=Add(A=@n7, B=@n15), Time=NamedRerouteUsage(), Speed=@n8) {speedY=1}) {parameterName=NoiseTexture02, samplerType=SAMPLERTYPE_LinearColor, texture=&T_NoisePack_LinearColor}
  n17 = LinearInterpolate(Alpha=Multiply(A=@n16.R) {constB=-1}) {constA=-1, constB=0.5}
  n18 = AppendVector(A=ComponentMask(None=TextureCoordinate()) {r=true}, B=Add(A=ComponentMask(None=@n14) {g=true}, B=Multiply(A=@n17, B=@n11) {constB=0.2}))
  n19 = Constant2Vector() {g=0.25}
  n20 = LinearInterpolate(
   A=TextureSampleParameter2D(UVs=Add(A=@n12, B=@n13) {constB=0}) {parameterName=BaseTexture, texture=/Game/Effects/Fort_Effects/Textures/Mask/T_HalfSphere_D}.R
   B=TextureSampleParameter2D(UVs=Add(A=@n18, B=@n19) {constB=0}) {parameterName=BaseTexture, texture=&T_HalfSphere_D}.R
  )
  n21 = Saturate(None=Multiply(A=Saturate(None=@n20), B=ScalarParameter() {defaultValue=1, parameterName=Alpha_Multiply}))
  n22 = Saturate(None=Power(Base=Multiply(A=@n1.A, B=@n21), Exp=ScalarParameter() {defaultValue=2, parameterName=Alpha_Power}))
  n23 = TextureCoordinate()
  n24 = Multiply(A=ComponentMask(None=@n23) {g=true}, B=OneMinus(None=ComponentMask(None=@n23) {g=true}))
  n25 = Saturate(None=Power(Base=Multiply(A=@n24) {constB=20}))
  n26 = ScalarParameter() {parameterName=MeshFade}
  DepthFade(
   Opacity=Multiply(
    A=Multiply(A=@n22, B=@n25)
    B=MaterialFunctionCall("FadeSensitivity (optional)"=@n26) {materialFunction=&FadeBasedOnViewAngle}.Result
   )
   FadeDistance=ScalarParameter() {defaultValue=50, parameterName=DepthFade}
  ) {fadeDistanceDefault=50}
  n27 = VectorParameter() {parameterName=Color, defaultValue=<rgba 0.0436 0 0.2917 1>}
  n28 = TextureSampleParameter2D(UVs=Panner(Coordinate=TextureCoordinate(), Time=NamedRerouteUsage()) {speedY=1}) {parameterName=ColorNoise, texture=/Game/Gadgets/Assets/VinderTech_GliderChute/Glider_Cube_Queen_Female/FX/T_Cheap_blue_space_color}
  n29 = Constant() {r=-1}
  n30 = Add(
   A=Multiply(
    A=@n27.RGB
    B=TextureSampleParameter2D() {parameterName=Noise, samplerType=SAMPLERTYPE_Grayscale, texture=&LowResBlurredNoise}.R
   )
   B=Add(A=@n27.RGB, B=Desaturation(None=@n28.RGB, Fraction=@n29)) {constB=0}
  )
  n31 = ScalarParameter() {parameterName=Des}
  MaterialFunctionCall(Emissive=Multiply(A=Desaturation(None=@n30, Fraction=@n31), B=@n1.RGB)) {materialFunction=&MF_FX_EmissiveAdjustments}
  NamedRerouteDeclaration(None=Multiply(A=Time(), B=ScalarParameter() {defaultValue=1, parameterName=TimeScale})) {name=CustomTime, nodeColor=<rgba 0.3874 0 1 1>}
 ]
}
