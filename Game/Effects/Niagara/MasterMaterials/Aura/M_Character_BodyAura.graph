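# Translucent unlit Niagara particle material, named for a character body aura, that renders animated, distorted glow shapes.
# Two half-sphere mask samples are vertically warped by per-particle randomized scrolling noise, summed, then broken up by caustics noise whose vertical speed scales with a dynamic parameter.
# Particle alpha and depth fade finish opacity.
# Color comes from scrolling color noise, hue-shifted, oversaturated through negative desaturation, tinted by particle color and passed through shared emissive adjustments.
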Material /Game/Effects/Niagara/MasterMaterials/Aura/M_Character_BodyAura
graph {
 properties {
  bUsedWithNiagaraMeshParticles=true
  bUsedWithNiagaraSprites=true
  blendMode=BLEND_Translucent
  pixelDepthOffsetMode=PDOM_Legacy
  shadingModel=MSM_Unlit
  usedShadingModels=Unlit
 }
 nodes [
  n1 = ParticleColor()
  n2 = TextureCoordinate()
  n3 = TextureCoordinate()
  n4 = Panner(Coordinate=TextureCoordinate()) {speedY=0.1}
  n5 = Multiply(
   A=ComponentMask(None=TextureSample(UVs=@n4) {samplerType=SAMPLERTYPE_Masks, texture=/Game/Characters/Player/Male/Medium/Bodies/M_MED_Storm_Soldier/Textures/ElectricityNoise_M}.RGB) {g=true, r=true}
   B=ScalarParameter() {parameterName=NoiseAmount}
  ) {constB=0}
  n6 = ParticleRandom()
  n7 = Add(A=Add(A=TextureCoordinate() {uTiling=1.3, vTiling=0.3}, B=@n5), B=@n6)
  n8 = Constant2Vector() {g=0.75}
  n9 = LinearInterpolate(Alpha=TextureSample(UVs=Panner(Coordinate=@n7, Speed=@n8) {speedY=1}) {samplerType=SAMPLERTYPE_LinearColor, texture=&T_NoisePack_LinearColor}.R) {constA=-1, constB=0.5}
  n10 = ScalarParameter() {defaultValue=0.2, parameterName=DistortionAmount1}
  n11 = AppendVector(A=Add(A=ComponentMask(None=@n2) {r=true}) {constB=0}, B=Add(A=ComponentMask(None=@n3) {g=true}, B=Multiply(A=@n9, B=@n10) {constB=0.2}))
  n12 = Constant2Vector() {g=0.1}
  n13 = TextureCoordinate()
  n14 = TextureCoordinate() {uTiling=1.3, vTiling=-0.3}
  n15 = Panner(Coordinate=Add(A=@n6, B=Add(A=@n5, B=@n14)), Speed=@n8) {speedY=1}
  n16 = LinearInterpolate(Alpha=Multiply(A=TextureSample(UVs=@n15) {samplerType=SAMPLERTYPE_LinearColor, texture=&T_NoisePack_LinearColor}.R) {constB=-1}) {constA=-1, constB=0.5}
  n17 = ScalarParameter() {defaultValue=0.2, parameterName=DistortionAmount2}
  n18 = AppendVector(A=ComponentMask(None=TextureCoordinate()) {r=true}, B=Add(A=ComponentMask(None=@n13) {g=true}, B=Multiply(A=@n16, B=@n17) {constB=0.2}))
  n19 = Constant2Vector() {g=0.25}
  n20 = Add(
   A=TextureSampleParameter2D(UVs=Add(A=@n11, B=@n12) {constB=0}) {parameterName=MainTexture, texture=/Game/Effects/Fort_Effects/Textures/Mask/T_HalfSphere_D}.R
   B=TextureSampleParameter2D(UVs=Add(A=@n18, B=@n19) {constB=0}) {parameterName=MainTexture, texture=&T_HalfSphere_D}.R
  ) {constB=0}
  n21 = Multiply(
   A=TextureCoordinate()
   B=AppendVector(A=ScalarParameter() {defaultValue=1, parameterName=Alpha noiseU}, B=ScalarParameter() {defaultValue=0.5, parameterName=Alpha noiseV})
  )
  n22 = ParticleRandom()
  n23 = ScalarParameter() {parameterName=Alpha noiseXspeed}
  n24 = Multiply(
   A=ScalarParameter() {defaultValue=1, parameterName=Alpha noiseYspeed}
   B=DynamicParameter() {paramNames [AlphaNoise_SpeedOffset, Param2, Param3, Param4]}.AlphaNoise_SpeedOffset
  )
  n25 = TextureSampleParameter2D(UVs=Panner(Coordinate=Add(A=@n21, B=@n22), Speed=AppendVector(A=@n23, B=@n24)) {speedY=1}) {
   parameterName=Alpha noise
   samplerType=SAMPLERTYPE_LinearGrayscale
   texture=/Game/Global/Textures/Default/Caustics/T_Nosie_Caustics_Grayscale
  }
  n26 = Multiply(A=LinearInterpolate(Alpha=Saturate(None=@n20)) {constA=-0.2}, B=Saturate(None=LinearInterpolate(Alpha=@n25.R)))
  DepthFade(Opacity=Power(Base=Multiply(A=@n1.A, B=@n26)), FadeDistance=ScalarParameter() {defaultValue=50, parameterName=DepthFade}) {fadeDistanceDefault=50}
  n27 = Panner() {speedY=1}
  n28 = Multiply(A=MaterialFunctionCall(
    Hue Shift Percentage=ScalarParameter() {parameterName=HueShiftPercent}
    Texture=TextureSample(UVs=@n27) {texture=/Game/Effects/Fort_Effects/Emotes/Energize_Max/LV4/T_Energize_Color_Noise_blur}.RGB
   ) {materialFunction=&HueShift}.Result) {constB=5}
  n29 = ScalarParameter() {defaultValue=-10, parameterName=Des}
  MaterialFunctionCall(Emissive=Multiply(A=Desaturation(None=@n28, Fraction=@n29), B=@n1.RGB)) {materialFunction=&MF_FX_EmissiveAdjustments}
 ]
}
